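# Material function that applies DBuffer decals directly to material attributes, replicating the automatic DBuffer Decal Response logic.
# Three inputs default to the DBuffer textures, with each alpha scaling the existing attributes before adding decal values.
# A drives base color and only attenuates subsurface color; B blends the world-space normal, then renormalizes.
# C's red, green and blue feed roughness, metallic and specular.
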MaterialFunction /Engine/Functions/Engine_MaterialFunctions01/Decal/ApplyDBuffer
graph {
 properties {
  bExposeToLibrary=true
  description=Apply DBuffer directly to material attributes. This replicates the logic automatically applied by the Dbuffer Decal Response material setting. Material normal should be in world space.
  libraryCategoriesText [Decals]
 }
 nodes [
  n1 = FunctionInput() {inputName=InMaterial, inputType=FunctionInput_MaterialAttributes}
  n2 = GetMaterialAttributes(None=@n1) {attributeGetTypes [
    #1
    #2
    #3
    #4
    #5
    #6
   ]}
  n3 = FunctionInput(Preview=DBufferTexture!(UVs).RGBA) {bUsePreviewValueAsDefault=true, inputName=DBufferA, inputType=FunctionInput_Vector4, sortPriority=1}
  n4 = ComponentMask(None=@n3) {a=true}
  n5 = FunctionInput(Preview=DBufferTexture() {dBufferTextureId=DBT_B}.RGBA) {bUsePreviewValueAsDefault=true, inputName=DBufferB, inputType=FunctionInput_Vector4, sortPriority=2}
  n6 = Multiply(A=@n2.Normal, B=ComponentMask(None=@n5) {a=true})
  n7 = ComponentMask(None=@n5) {b=true, g=true, r=true}
  n8 = MaterialFunctionCall(Float4=FunctionInput(Preview=DBufferTexture() {dBufferTextureId=DBT_C}.RGBA) {bUsePreviewValueAsDefault=true, inputName=DBufferC, inputType=FunctionInput_Vector4, sortPriority=3}) {materialFunction=&BreakOutFloat4Components}
  n9 = SetMaterialAttributes!(
   MaterialAttributes=@n1
   BaseColor=Add(A=Multiply(A=@n2.BaseColor, B=@n4), B=ComponentMask(None=@n3) {b=true, g=true, r=true})
   SubsurfaceColor=Multiply(A=@n2.SubsurfaceColor, B=@n4)
   Normal=Normalize(VectorInput=Add(A=@n6, B=@n7))
   Metallic=Add(A=Multiply(A=@n2.Metallic, B=@n8.A), B=@n8.G)
   Specular=Add(A=Multiply(A=@n2.Specular, B=@n8.A), B=@n8.B)
   Roughness=Add(A=Multiply(A=@n2.Roughness, B=@n8.A), B=@n8.R)
  ) {attributeSetTypes [
    #1
    #2
    #3
    #4
    #5
    #6
   ]}
  FunctionOutput(None=@n9) {outputName=OutMaterial}
 ]
}
